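# Parent material for the CherryKnot backpack, built on the 2023 character parent material functions with an Ice CrewStyle effect layered on top.
# PreFX output runs through PostFX, then into the Ice function along with the PreFX Fresnel.
# That function's TechArt input packs the blue Tech Art Mask channel, the red channel of a backpack FX texture and a constant.
# It uses material attributes, with the shading model set by expression.
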Material /BRCosmetics/Accessories/FORT_Backpacks/Backpack_F_MED_CherryKnot/Materials/M_CherryKnot_Backpack_Parent
graph {
 properties {
  bAutomaticallySetUsageInEditor=false
  bUseMaterialAttributes=true
  bUsedWithMeshDeformer=true
  bUsedWithMorphTargets=true
  bUsedWithSkeletalMesh=true
  pixelDepthOffsetMode=PDOM_Legacy
  shadingModel=MSM_FromMaterialExpression
  usedShadingModels=Default Lit | Subsurface | Clear Coat | Subsurface Profile | Cloth
 }
 nodes [
  n1 = MaterialFunctionCall() {materialFunction=&MF_CharacterParent_2023_PreFX}
  n2 = ComponentMask(None=@n1.Tech Art Mask) {b=true}
  n3 = TextureSample() {samplerType=SAMPLERTYPE_Masks, texture=/BRCosmetics/Accessories/FORT_Backpacks/Backpack_F_MED_CherryKnot/Textures/T_CherryKnot_Backpack_FX_2}
  MaterialFunctionCall!(
   TechArt=AppendVector(A=AppendVector(A=@n2, B=@n3.R), B=Constant())
   Fresnel=@n1.Fresnel
   In=MaterialFunctionCall(Material=@n1.Result, M=@n1.Result, Pre Skinned Local Position=@n1.Result, Distance Blend=@n1.Result, Fresnel=@n1.Result) {materialFunction=&MF_CharacterParent_2023_PostFX}.Result
  ) {materialFunction=/Game/Global/Materials/Character_Parent_2023/CrewStyle_Parent_2023/Ice/MF_CrewStyle_Ice_Inst}
 ]
}
