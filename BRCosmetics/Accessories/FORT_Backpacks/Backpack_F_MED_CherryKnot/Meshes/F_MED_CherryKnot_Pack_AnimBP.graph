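# Anim Blueprint for the CherryKnot backpack, adding physics sway to its chain and lid bones and playing an elimination reaction animation.
# Front-end and in-game state machines feed a cached pose, triggered by timed delays and enemy-elimination events respectively.
# That pose drives separately tuned AnimDynamics chains, selected by bIsInFrontEnd.
# The in-game branch also rotates base_lid and swaps chain sim-space settings while skydiving.
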AnimBlueprint /BRCosmetics/Accessories/FORT_Backpacks/Backpack_F_MED_CherryKnot/Meshes/F_MED_CherryKnot_Pack_AnimBP
target_skeleton=&/F_MED_CherryKnot_Pack_Skeleton
graphs [
 AnimGraph {nodes [
   Input Pose()
   n1 = Anim Dynamics - Bone: dyn_chain_01(ComponentPose=Local To Component(LocalPose=Use cached pose 'input pose'()), Alpha=1, ExternalForce="0,0,0") {
    angularDampingOverride=1
    angularSpringConstant=12000
    bAngularSpring=true
    bOverrideAngularDamping=true
    bUseGravityOverride=true
    gravityScale=0.35
    simulationSpace=RootRelative
    boundBone {boneName=dyn_chain_01}
    componentAppliedLinearAccClamp=<5000 5000 5000>
    componentLinearAccScale=<0.9 0.9 0.9>
    componentLinearVelScale=<0.9 0.9 0.9>
    gravityOverride {z=-980}
    physicsBodyDefinitions [{
      boundBone {boneName=dyn_chain_01}
      boxExtents=<20 30 30>
      constraintSetup {angularLimitsMax=<0 1 2>, angularLimitsMin=<0 -1 -18>, angularTarget=<1 0 0>}
      localJointOffset=<3.9 0 0>
     }]
   }
   n2 = Anim Dynamics - Bone: dyn_chain_02(ComponentPose=@n1, Alpha=1, ExternalForce="0,0,0") {
    angularDampingOverride=1
    angularSpringConstant=18000
    bAngularSpring=true
    bOverrideAngularDamping=true
    bUseGravityOverride=true
    gravityScale=0.35
    simulationSpace=RootRelative
    boundBone {boneName=dyn_chain_02}
    componentAppliedLinearAccClamp=<6000 6000 6000>
    componentLinearAccScale=<0.9 0.9 0.9>
    componentLinearVelScale=<0.9 0.9 0.9>
    gravityOverride {z=-988}
    physicsBodyDefinitions [{
      boundBone {boneName=dyn_chain_02}
      boxExtents=<27 27 27>
      constraintSetup {angularLimitsMax=<10 18 1>, angularLimitsMin=<-10 -18 -17>, angularTarget=<1 0 0>}
      localJointOffset=<4.2 0 0>
     }]
   }
   n3 = Anim Dynamics - Bone: dyn_chain_03(ComponentPose=@n2, Alpha=1, ExternalForce="0,0,0") {
    angularDampingOverride=1
    angularSpringConstant=16000
    bAngularSpring=true
    bOverrideAngularDamping=true
    bUseGravityOverride=true
    gravityScale=0.35
    simulationSpace=RootRelative
    boundBone {boneName=dyn_chain_03}
    componentAppliedLinearAccClamp=<6000 6000 6000>
    componentLinearAccScale=<0.9 0.9 0.9>
    componentLinearVelScale=<0.9 0.9 0.9>
    gravityOverride {z=-980}
    physicsBodyDefinitions [{
      boundBone {boneName=dyn_chain_03}
      boxExtents=<28 28 28>
      constraintSetup {angularLimitsMax=<10 17 2>, angularLimitsMin=<-10 -17 -16>, angularTarget=<1 0 0>}
      localJointOffset=<4.1 0 0>
     }]
   }
   n4 = Anim Dynamics - Bone: dyn_chain_04(ComponentPose=@n3, Alpha=1, ExternalForce="0,0,0") {
    angularDampingOverride=1
    angularSpringConstant=18000
    bAngularSpring=true
    bOverrideAngularDamping=true
    bUseGravityOverride=true
    gravityScale=0.35
    simulationSpace=RootRelative
    boundBone {boneName=dyn_chain_04}
    componentAppliedLinearAccClamp=<6000 6000 6000>
    componentLinearAccScale=<0.9 0.9 0.9>
    componentLinearVelScale=<0.9 0.9 0.9>
    gravityOverride {z=-980}
    physicsBodyDefinitions [{
      boundBone {boneName=dyn_chain_04}
      boxExtents=<28 28 28>
      constraintSetup {angularLimitsMax=<10 15 1>, angularLimitsMin=<-10 -15 -15>, angularTarget=<1 0 0>}
      localJointOffset=<4.1 0 0>
     }]
   }
   n5 = Anim Dynamics - Bone: dyn_lid_01(ComponentPose=@n4, Alpha=0.6, ExternalForce="0,0,0") {
    angularDampingOverride=1
    angularSpringConstant=16000
    bAngularSpring=true
    bOverrideAngularDamping=true
    bUseGravityOverride=true
    gravityScale=0.35
    simulationSpace=RootRelative
    boundBone {boneName=dyn_lid_01}
    componentAppliedLinearAccClamp=<4000 4000 4000>
    componentLinearAccScale=<0.6 0.6 0.6>
    componentLinearVelScale=<0.6 0.6 0.6>
    gravityOverride {z=-960}
    physicsBodyDefinitions [{
      boundBone {boneName=dyn_lid_01}
      boxExtents=<20 30 30>
      constraintSetup {angularLimitsMax=<0 26 0>, angularLimitsMin=<0 -26 0>, angularTarget=<1 0 0>}
      localJointOffset=<3.7 0 0>
     }]
   }
   n6 = Get Sim Space Settings() -> (Sim Space Settings:Anim Phys Sim Space Settings Structure)
   n7 = Anim Dynamics - Bone: dyn_chain_01!(
    ComponentPose=Local To Component(LocalPose=Use cached pose 'input pose'())
    Alpha=1
    SimSpaceSettings:Anim Phys Sim Space Settings Structure=@n6
    ExternalForce="0,0,0"
   ) -> (Pose) {
    angularDampingOverride=1
    angularSpringConstant=12000
    bAngularSpring=true
    bOverrideAngularDamping=true
    bUseGravityOverride=true
    gravityScale=0.35
    simulationSpace=RootRelative
    boundBone {boneName=dyn_chain_01}
    componentAppliedLinearAccClamp=<5000 5000 5000>
    componentLinearAccScale=<0.7 0.7 0.7>
    componentLinearVelScale=<0.7 0.7 0.7>
    gravityOverride {z=-980}
    physicsBodyDefinitions [{
      boundBone {boneName=dyn_chain_01}
      boxExtents=<20 29 29>
      constraintSetup {angularLimitsMax=<0 1 2>, angularLimitsMin=<0 -1 -18>, angularTarget=<1 0 0>}
      localJointOffset=<3.7 0 0>
     }]
   }
   n8 = Anim Dynamics - Bone: dyn_chain_02!(ComponentPose=@n7, Alpha=1, SimSpaceSettings:Anim Phys Sim Space Settings Structure=@n6, ExternalForce="0,0,0") -> (Pose) {
    angularDampingOverride=1
    angularSpringConstant=18000
    bAngularSpring=true
    bOverrideAngularDamping=true
    bUseGravityOverride=true
    gravityScale=0.35
    simulationSpace=RootRelative
    boundBone {boneName=dyn_chain_02}
    componentAppliedLinearAccClamp=<5000 5000 5000>
    componentLinearAccScale=<0.8 0.8 0.8>
    componentLinearVelScale=<0.85 0.85 0.85>
    gravityOverride {z=-988}
    physicsBodyDefinitions [{
      boundBone {boneName=dyn_chain_02}
      boxExtents=<20 28 28>
      constraintSetup {angularLimitsMax=<10 18 1>, angularLimitsMin=<-10 -18 -19>, angularTarget=<1 0 0>}
      localJointOffset=<3.9 0 0>
     }]
    simSpaceSettings {maxAngularAcceleration=1000, maxAngularVelocity=100}
   }
   n9 = Anim Dynamics - Bone: dyn_chain_03!(ComponentPose=@n8, Alpha=1, SimSpaceSettings:Anim Phys Sim Space Settings Structure=@n6, ExternalForce="0,0,0") -> (Pose) {
    angularDampingOverride=1
    angularSpringConstant=16000
    bAngularSpring=true
    bOverrideAngularDamping=true
    bUseGravityOverride=true
    gravityScale=0.35
    simulationSpace=RootRelative
    boundBone {boneName=dyn_chain_03}
    componentAppliedLinearAccClamp=<5000 5000 5000>
    componentLinearAccScale=<0.85 0.85 0.85>
    componentLinearVelScale=<0.85 0.85 0.85>
    gravityOverride {z=-980}
    physicsBodyDefinitions [{
      boundBone {boneName=dyn_chain_03}
      boxExtents=<20 29 29>
      constraintSetup {angularLimitsMax=<10 18 2>, angularLimitsMin=<-10 -18 -18>, angularTarget=<1 0 0>}
      localJointOffset=<3.8 0 0>
     }]
    simSpaceSettings {maxAngularAcceleration=1000, maxAngularVelocity=100}
   }
   n10 = Anim Dynamics - Bone: dyn_chain_04!(ComponentPose=@n9, Alpha=1, SimSpaceSettings:Anim Phys Sim Space Settings Structure=@n6, ExternalForce="0,0,0") -> (Pose) {
    angularDampingOverride=1
    angularSpringConstant=18000
    bAngularSpring=true
    bOverrideAngularDamping=true
    bUseGravityOverride=true
    gravityScale=0.35
    simulationSpace=RootRelative
    boundBone {boneName=dyn_chain_04}
    componentAppliedLinearAccClamp=<5000 5000 5000>
    componentLinearAccScale=<0.9 0.9 0.9>
    componentLinearVelScale=<0.9 0.9 0.9>
    gravityOverride {z=-980}
    physicsBodyDefinitions [{
      boundBone {boneName=dyn_chain_04}
      boxExtents=<20 29 29>
      constraintSetup {angularLimitsMax=<10 18 1>, angularLimitsMin=<-10 -18 -19>, angularTarget=<1 0 0>}
      localJointOffset=<3.8 0 0>
     }]
    simSpaceSettings {maxAngularAcceleration=1000, maxAngularVelocity=100}
   }
   n11 = Anim Dynamics - Bone: dyn_lid_01(ComponentPose=@n10, Alpha=1, ExternalForce="0,0,0") {
    angularDampingOverride=1
    angularSpringConstant=16500
    bAngularSpring=true
    bOverrideAngularDamping=true
    bUseGravityOverride=true
    gravityScale=0.35
    simulationSpace=RootRelative
    boundBone {boneName=dyn_lid_01}
    componentAppliedLinearAccClamp=<4500 4500 4500>
    componentLinearAccScale=<0.55 0.55 0.55>
    componentLinearVelScale=<0.6 0.6 0.6>
    gravityOverride {z=-980}
    physicsBodyDefinitions [{
      boundBone {boneName=dyn_lid_01}
      boxExtents=<20 29.5 29.5>
      constraintSetup {angularLimitsMax=<0 45 0>, angularLimitsMin=<0 -32 0>, angularTarget=<1 0 0>}
      localJointOffset=<3 0 0>
     }]
    simSpaceSettings {maxAngularAcceleration=1000, maxAngularVelocity=100, simSpaceAngularAlpha=0.15}
   }
   n12 = "Transform (Modify) Bone - Bone: base_lid"!(ComponentPose=@n11, bAlphaBoolEnabled:bool=True, Translation="0,0,0", Rotation="-65,0,0", Scale="1,1,1") -> (Pose) {
    alphaInputType=Bool
    rotationMode=BMM_Additive
    rotationSpace=BCS_BoneSpace
    alphaBoolBlend {blendInTime=0.6, blendOutTime=0.3}
    boneToModify {boneName=base_lid}
    rotation {pitch=-65}
   }
   Output Pose(Result=Blend Poses by bool(BlendPose_0=Component To Local(ComponentPose=@n5), BlendPose_1=Component To Local(ComponentPose=@n12), bActiveValue=Get bIsInFrontEnd())) {layerGroup=None}
   AnimGraphNode_StateMachine_0 = Idle and Reactive() {is=title AnimGraphNode_StateMachine}
   AnimGraphNode_StateMachine_1 = FE_Reactivity() {is=title AnimGraphNode_StateMachine}
   Save cached pose 'input pose'(Pose=Blend Poses by bool(BlendPose_0=@AnimGraphNode_StateMachine_1, BlendPose_1=@AnimGraphNode_StateMachine_0, bActiveValue=Get bIsInFrontEnd()) {blendType=Linear}) {cacheName=input pose}
  ]}
 Idle and Reactive {
  entry=Idle
  path=AnimGraph.AnimGraphNode_StateMachine_0.Idle and Reactive
  ids=id
  nodes [
   AnimStateNode_1 = Reactive_Elimination() {
    is=title AnimStateNode
    stateEntered {guid=#1}
    stateFullyBlended {guid=#2}
    stateLeft {guid=#3}
   }
   AnimStateNode_2 = Idle() {
    stateEntered {guid=#4}
    stateFullyBlended {guid=#5}
    stateLeft {guid=#6}
   }
   AnimStateTransitionNode_2 = AnimStateTransitionNode() {
    crossfadeDuration=0
    transitionEnd {guid=#7}
    transitionInterrupt {guid=#8}
    transitionStart {guid=#9}
   }
   AnimStateTransitionNode_3 = AnimStateTransitionNode() {
    bAutomaticRuleBasedOnSequencePlayerInState=true
    crossfadeDuration=0
    transitionEnd {guid=#10}
    transitionInterrupt {guid=#11}
    transitionStart {guid=#12}
   }
  ]
  links [
   {
    dst=Reactive_Elimination
    rule=AnimGraph.AnimGraphNode_StateMachine_0.Idle and Reactive.AnimStateTransitionNode_2.Transition
    src=Idle
    via=AnimStateTransitionNode_2
   }
   {
    dst=Idle
    rule=AnimGraph.AnimGraphNode_StateMachine_0.Idle and Reactive.AnimStateTransitionNode_3.Transition
    src=Reactive_Elimination
    via=AnimStateTransitionNode_3
   }
  ]
 }
 Reactive_Elimination {
  path=AnimGraph.AnimGraphNode_StateMachine_0.Idle and Reactive.AnimStateNode_1.Reactive_Elimination
  nodes [
   n1 = Sequence Player 'CherryKnot_ElimReact_PACK'() {bLoopAnimation=false, sequence=/BRCosmetics/Accessories/FORT_Backpacks/Backpack_F_MED_CherryKnot/Animation/CherryKnot_ElimReact_PACK}
   Output Animation Pose(Result=@n1)
  ]
 }
 Idle {
  path=AnimGraph.AnimGraphNode_StateMachine_0.Idle and Reactive.AnimStateNode_2.Idle
  nodes [Output Animation Pose(Result=Sequence Player 'CherryKnot_BasePose_PACK'() {bLoopAnimation=false, sequence=/BRCosmetics/Accessories/FORT_Backpacks/Backpack_F_MED_CherryKnot/Animation/CherryKnot_BasePose_PACK})]
 }
 Transition {
  path=AnimGraph.AnimGraphNode_StateMachine_0.Idle and Reactive.AnimStateTransitionNode_2.Transition
  nodes [Result(bCanEnterTransition=Get Reactive Elimination() -> (Reactive Elimination:bool))]
 }
 Transition {path=AnimGraph.AnimGraphNode_StateMachine_0.Idle and Reactive.AnimStateTransitionNode_3.Transition, nodes [Result()]}
 FE_Reactivity {
  entry=Idle
  path=AnimGraph.AnimGraphNode_StateMachine_1.FE_Reactivity
  ids=id
  nodes [
   AnimStateNode_0 = Idle()
   AnimStateNode_2 = FE_Reactive() {is=title AnimStateNode, stateEntered {guid=#13, notifyName=Entered FE_Reactive_Anim}}
   AnimStateTransitionNode_8 = AnimStateTransitionNode() {crossfadeDuration=0}
   AnimStateTransitionNode_9 = AnimStateTransitionNode() {bAutomaticRuleBasedOnSequencePlayerInState=true, crossfadeDuration=0}
  ]
  links [
   {
    dst=Idle
    rule=AnimGraph.AnimGraphNode_StateMachine_1.FE_Reactivity.AnimStateTransitionNode_9.Transition
    src=FE_Reactive
    via=AnimStateTransitionNode_9
   }
   {
    dst=FE_Reactive
    rule=AnimGraph.AnimGraphNode_StateMachine_1.FE_Reactivity.AnimStateTransitionNode_8.Transition
    src=Idle
    via=AnimStateTransitionNode_8
   }
  ]
 }
 Idle {
  path=AnimGraph.AnimGraphNode_StateMachine_1.FE_Reactivity.AnimStateNode_0.Idle
  nodes [Output Animation Pose(Result=Sequence Player 'CherryKnot_BasePose_PACK'() {bLoopAnimation=false, sequence=&CherryKnot_BasePose_PACK})]
 }
 FE_Reactive {
  path=AnimGraph.AnimGraphNode_StateMachine_1.FE_Reactivity.AnimStateNode_2.FE_Reactive
  nodes [
   n1 = Sequence Player 'CherryKnot_ElimReact_PACK'() {bLoopAnimation=false, sequence=&CherryKnot_ElimReact_PACK}
   Output Animation Pose(Result=@n1)
  ]
 }
 Transition {
  path=AnimGraph.AnimGraphNode_StateMachine_1.FE_Reactivity.AnimStateTransitionNode_8.Transition
  nodes [Result(bCanEnterTransition=Get Play Reactive FE Anim() -> (Play Reactive FE Anim:bool))]
 }
 Transition {path=AnimGraph.AnimGraphNode_StateMachine_1.FE_Reactivity.AnimStateTransitionNode_9.Transition, nodes [Result()]}
 EventGraph {nodes [
   Event OnCharacterCustomizationLoaded()
   > n1 = Cast To FortPlayerPawn(Object=TryGetFortPawn())
   > n2 = Branch()
   > Do Once(Reset=@n6.then)
   > n3 = Delay(Duration=3)
   Set Play Reactive FE Anim(execute=[@n3.then, @n5.then], Play Reactive FE Anim:bool=true) -> (Output_Get:bool)
   > n4 = Delay(Duration=1)
   Set Play Reactive FE Anim(execute=[@n4.then, @n8.then], Play Reactive FE Anim=false)
   > n5 = Delay(Duration=15)
   n6 = Set Play Reactive FE Anim(execute=@n2.else, Play Reactive FE Anim=false)
   Event BlueprintUpdateAnimation()
   > Sequence()
   > n7 = Branch(Condition=Get bIsSkyDiving())
   > Set Sim Space Settings(Sim Space Settings:Anim Phys Sim Space Settings Structure=Get Sim Space Settings skydive() -> (Sim Space Settings skydive:Anim Phys Sim Space Settings Structure)) -> (Output_Get:Anim Phys Sim Space Settings Structure)
   Set Sim Space Settings(execute=@n7.else, Sim Space Settings=Get Sim Space Settings ingame() -> (Sim Space Settings ingame:Anim Phys Sim Space Settings Structure))
   n8 = AnimNotify_FE_ReactiveAnim() {is=title K2Node_Event}
   n9 = Event BlueprintProcessAnimInputEvent()
   > Branch(Condition=IsNamed(self=@n9.AnimInputEvent, InEventName=aie.EnemyPlayerEliminated))
   > Set Reactive Elimination(Reactive Elimination:bool=true) -> (Output_Get:bool)
   > RetriggerableDelay(Duration=0.2)
   > Set Reactive Elimination(Reactive Elimination=false)
   TryGetPawnOwner()
   GetModifierForPartType!(self:Fort Player Pawn Object Reference=@n1, PartType:EFortCustomPartType Enum=Backpack) -> (ReturnValue:Custom Character Part Modifier Object Reference)
  ]}
]
variables [
 JumpFall {type=boolean}
 Play Reactive FE Anim {type=boolean}
 Reactive Elimination {type=boolean}
 Sim Space Settings {
  type=AnimPhysSimSpaceSettings
  default {maxAngularAcceleration=10000, maxAngularVelocity=10000, simSpaceAngularAlpha=0, externalAngularVelocity=<0 0 0>}
 }
 Sim Space Settings ingame {
  type=AnimPhysSimSpaceSettings
  default {maxAngularAcceleration=1000, maxAngularVelocity=15, simSpaceAngularAlpha=1, externalAngularVelocity=<0 0 0>}
 }
 Sim Space Settings skydive {
  type=AnimPhysSimSpaceSettings
  default {maxAngularAcceleration=1000, maxAngularVelocity=100, simSpaceAngularAlpha=0, externalAngularVelocity=<0 0 0>}
 }
]
